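; Season 13 quest ability that credits a player for landing a Hoagie, apparently at the bottom of a smokestack.
; On event activation it ends at once unless the owning pawn has the in-Hoagie tag, otherwise it starts a looping timer polling whether the vehicle touches the ground.
; Touchdown sends a Landed stat event to the quest manager and ends the ability.
; Losing the SmokeStack_Bottom tag first cancels the timer and ends it without credit.
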(event Ability|EventActivateAbilityFromEvent
  (params (EventData :type Gameplay%20Event%20Data%20Structure%20%28by%20ref%29))
  (|CallParentFunction EventData)
  (if (GameplayTags|TagContainer|CompareTagAssetInterfacetoOtherTags (Variables|Default|GetT_InHoagie) (Variables|GAVars|Data|GetOwningPlayerPawn))
    (Development|PrintString "Ability Started" true true "(R=0.000000,G=0.660000,B=1.000000,A=1.000000)" 1e+01)
    (bind _returnvalue (Utilities|Time|SetTimerbyEvent (delegate Custom|check%20if%20landed) 2.0 true))
    (Variables|Default|SetTimer%20Handle _returnvalue)
    (Ability|Tasks|WaitGameplayTagRemove (Variables|Default|GetT_SmokeStack_Bottom)
      (:Removed
        (Utilities|Time|ClearandInvalidateTimerbyHandle (Variables|Default|GetTimer%20Handle))
        (Ability|EndAbility)))
    (else
      (Ability|EndAbility))))

(event Custom|check%20if%20landed
  (if (Class|FortVehicleInterface|IsTouchingGround (Vehicle|GetVehicleInterface (Variables|GAVars|Data|GetOwningPlayerPawn)))
    (Development|PrintString "Landed")
    (Quest|SendComplexCustomStatEvent (Variables|GAVars|Data|GetOwningPlayerQuestManager) 0 (Variables|Default|GetT_Landed) "(GameplayTags=)" 1)
    (Utilities|Time|ClearandInvalidateTimerbyHandle (Variables|Default|GetTimer%20Handle))
    (Ability|EndAbility)
    (else
      (Development|PrintString "Not Landed"))))
